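# MetaSound patch in the Harmonix test content that plays a short enveloped sine tone at a given MIDI note on each trigger.
# It converts the MIDI note to hertz with equal-temperament math relative to A440, driving a sine oscillator.
# The input trigger starts an ADSR envelope and, through a short trigger delay, releases it automatically, so each trigger yields a fixed pulse.
# The envelope multiplies the sine to form the audio output.
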MetaSoundPatch /Harmonix/Tests/HarmonixMetasound/MS_TriggerPulseSineGenerator
inputs [
 MIDI Note {type=Int32}
 Trigger {type=Trigger}
]
outputs [Out {type=Audio}]
pages [Default {
  input_defaults {MIDI Note=60, Trigger=null}
  nodes [
   TriggerDelay_0 = UE.Trigger Delay(In=^Trigger) {also {defaults {Delay Time=0.1}}}
   ADSREnvelopeADSREnvelope_0 = ADSR Envelope.ADSR Envelope:Audio(Trigger Attack=^Trigger, Trigger Release=@TriggerDelay_0.Out) {also {defaults {Release Time=0.1}}}
   n1 = UE.Subtract:Int32(PrimaryOperand=^MIDI Note) {also {defaults {AdditionalOperands=69}}}
   n2 = UE.Power:Float(AdditionalOperands=UE.Divide:Float(PrimaryOperand=Convert.Float:Int32(Int32=@n1.Out).Float) {also {defaults {AdditionalOperands=12}}}.Out) {also {defaults {PrimaryOperand=2}}}
   Multiply_0 = UE.Multiply:Audio(
    PrimaryOperand=@ADSREnvelopeADSREnvelope_0.Out Envelope
    AdditionalOperands=UE.Sine:Audio(Frequency=UE.Multiply:Float(PrimaryOperand=@n2.Out) {also {defaults {AdditionalOperands=440}}}.Out).Audio
   )
  ]
  links [{dst="@Out", src=Multiply_0.Out}]
 }]
