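# Harmonix test MetaSound, a mono one-shot source that exercises a MIDI Player driven by a Metronome's MIDI clock alongside clock-based triggers.
# OnPlay starts a shared transport, and the player's MIDI stream feeds a sine generator.
# A subdivision trigger and four bar-timestamp triggers fire sine pulses, the latter taking notes from an array.
# All voices are mixed to mono, and a delayed OnPlay trigger signals finish.
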MetaSoundSource /Harmonix/Tests/HarmonixMetasound/MidiPlayerNode/MS_Test_MidiPlayerNode_Metronome_MidiClockTriggers
inputs [
 Metronome.Speed {type=Float}
 Metronome.Tempo {type=Float}
 MIDI Player.Kill Voices On MIDI File Change {type=Bool, access=Constructor}
 MIDI Player.Kill Voices On Seek {type=Bool, access=Constructor}
 MIDI Player.MIDI File {type=MIDIAsset}
 MIDI Player.Pre-roll Bars {type=Int32, access=Constructor}
 MIDI Player.Simple Loop {type=Bool}
 MIDI Player.Speed {type=Float}
 Subdivision.MIDI Note {type=Int32}
 Timestamps.MIDI Note {type=Int32:Array}
 UE.Source.OnPlay {type=Trigger}
]
interfaces [UE.OutputFormat.Mono:1.0, UE.Source.OneShot:1.0, UE.Source:1.1]
outputs [
 AudioOut {type=Audio}
 "UE.OutputFormat.Mono.Audio:0" {type=Audio}
 UE.Source.OneShot.OnFinished {type=Trigger}
]
pages [Default {
  input_defaults {
   MIDI Player.Kill Voices On MIDI File Change=null
   MIDI Player.Kill Voices On Seek=null
   MIDI Player.MIDI File=&/scale_up_down
   MIDI Player.Pre-roll Bars=8
   MIDI Player.Simple Loop=null
   MIDI Player.Speed=1
   Metronome.Speed=1
   Metronome.Tempo=120
   Subdivision.MIDI Note=60
   UE.Source.OnPlay=null
   Timestamps.MIDI Note [72,74,76,77]
  }
  nodes [
   n1 = HarmonixNodes.TriggerToTransport(Play=^UE.Source.OnPlay)
   n2 = HarmonixNodes.Metronome(Speed=^Metronome.Speed, Tempo=^Metronome.Tempo, Transport=@n1.Transport)
   n3 = HarmonixNodes.MidiClockTimestampTrigger(MIDI Clock=@n2.MIDI Clock, Music Timestamp=HarmonixNodes.MusicTimestampBuilder() {also {defaults {Bar=2}, version [0,1]}}.Music Timestamp)
   n4 = HarmonixNodes.MidiClockTimestampTrigger(MIDI Clock=@n2.MIDI Clock, Music Timestamp=HarmonixNodes.MusicTimestampBuilder() {also {version [0,1]}}.Music Timestamp)
   n5 = HarmonixNodes.MidiClockTimestampTrigger(MIDI Clock=@n2.MIDI Clock, Music Timestamp=HarmonixNodes.MusicTimestampBuilder() {also {defaults {Bar=4}, version [0,1]}}.Music Timestamp)
   n6 = HarmonixNodes.MidiClockTimestampTrigger(MIDI Clock=@n2.MIDI Clock, Music Timestamp=HarmonixNodes.MusicTimestampBuilder() {also {defaults {Bar=3}, version [0,1]}}.Music Timestamp)
   AudioMixerAudioMixerMono5_0 = "AudioMixer.Audio Mixer (Mono, 5)"(
    In 1=/Harmonix/Tests/HarmonixMetasound/MS_TriggerPulseSineGenerator(
     MIDI Note=Array.Get:Int32:Array(Array=^Timestamps.MIDI Note, Trigger=@n3.Trigger Out) {also {defaults {Index=1}}}.Element
     Trigger=@n3.Trigger Out
    ).Out
    In 0=&MS_TriggerPulseSineGenerator(
     MIDI Note=Array.Get:Int32:Array(Array=^Timestamps.MIDI Note, Trigger=@n4.Trigger Out) {also {defaults {Index=0}}}.Element
     Trigger=@n4.Trigger Out
    ).Out
    In 4=&MS_TriggerPulseSineGenerator(
     MIDI Note=Array.Get:Int32:Array(Array=^Timestamps.MIDI Note, Trigger=@n5.Trigger Out) {also {defaults {Index=3}}}.Element
     Trigger=@n5.Trigger Out
    ).Out
    In 2=&MS_TriggerPulseSineGenerator(
     MIDI Note=Array.Get:Int32:Array(Array=^Timestamps.MIDI Note, Trigger=@n6.Trigger Out) {also {defaults {Index=2}}}.Element
     Trigger=@n6.Trigger Out
    ).Out
   )
   AudioMixerAudioMixerMono3_0 = "AudioMixer.Audio Mixer (Mono, 3)"(
    In 2=@AudioMixerAudioMixerMono5_0.Out
    In 0=/Harmonix/Tests/HarmonixMetasound/MS_MIDIStreamSineGenerator(MIDI Stream=HarmonixNodes.MIDIPlayer(
      Kill Voices On MIDI File Change=^MIDI Player.Kill Voices On MIDI File Change
      Kill Voices On Seek=^MIDI Player.Kill Voices On Seek
      MIDI File=^MIDI Player.MIDI File
      Pre-roll Bars=^MIDI Player.Pre-roll Bars
      Simple Loop=^MIDI Player.Simple Loop
      Speed=^MIDI Player.Speed
      MIDI Clock=@n2.MIDI Clock
      Transport=@n1.Transport
     ).MIDI Stream).Out
    In 1=&MS_TriggerPulseSineGenerator(MIDI Note=^Subdivision.MIDI Note, Trigger=HarmonixNodes.MidiClockSubdivisionTrigger(MIDI Clock=@n2.MIDI Clock).Trigger Out).Out
   )
   TriggerDelay_0 = UE.Trigger Delay(In=^UE.Source.OnPlay) {also {defaults {Delay Time=8}}}
  ]
  links [
   {dst="@AudioOut", src=AudioMixerAudioMixerMono3_0.Out}
   {dst="@UE.OutputFormat.Mono.Audio:0", src=AudioMixerAudioMixerMono3_0.Out}
   {dst="@UE.Source.OneShot.OnFinished", src=TriggerDelay_0.Out}
  ]
 }]
properties {
 assetImportData=&self:AssetImportData
 sampleRate=48000
}
