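# Mono one-shot MetaSound test source for the Harmonix MIDI Player node, driven by a Metronome clock through scripted tempo changes.
# On play, staggered trigger delays select tempos through a float trigger route that feeds the Metronome, which shares a transport with the player.
# The player's MIDI stream, defaulting to an up-down scale file, drives a sine generator sub-graph for audio.
# A final delay signals OnFinished after eight seconds.
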MetaSoundSource /Harmonix/Tests/HarmonixMetasound/MidiPlayerNode/MS_Test_MidiPlayerNode_Metronome_TempoChanges
inputs [
 Metronome.Speed {type=Float}
 Metronome.Tempo {type=Float}
 MIDI Player.Kill Voices On MIDI File Change {type=Bool, access=Constructor}
 MIDI Player.Kill Voices On Seek {type=Bool, access=Constructor}
 MIDI Player.MIDI File {type=MIDIAsset}
 MIDI Player.Pre-roll Bars {type=Int32, access=Constructor}
 MIDI Player.Simple Loop {type=Bool}
 MIDI Player.Speed {type=Float}
 UE.Source.OnPlay {type=Trigger}
]
interfaces [UE.OutputFormat.Mono:1.0, UE.Source.OneShot:1.0, UE.Source:1.1]
outputs [
 AudioOut {type=Audio}
 "UE.OutputFormat.Mono.Audio:0" {type=Audio}
 UE.Source.OneShot.OnFinished {type=Trigger}
]
pages [Default {
  input_defaults {
   MIDI Player.Kill Voices On MIDI File Change=null
   MIDI Player.Kill Voices On Seek=null
   MIDI Player.MIDI File=&/scale_up_down
   MIDI Player.Pre-roll Bars=8
   MIDI Player.Simple Loop=null
   MIDI Player.Speed=1
   Metronome.Speed=1
   Metronome.Tempo=120
   UE.Source.OnPlay=null
  }
  nodes [
   n1 = HarmonixNodes.TriggerToTransport(Play=^UE.Source.OnPlay)
   TriggerDelay_0 = UE.Trigger Delay(In=^UE.Source.OnPlay) {also {defaults {Delay Time=2}}}
   TriggerDelay_1 = UE.Trigger Delay(In=^UE.Source.OnPlay) {also {defaults {Delay Time=6}}}
   TriggerDelay_2 = UE.Trigger Delay(In=^UE.Source.OnPlay) {also {defaults {Delay Time=5}}}
   TriggerDelay_3 = UE.Trigger Delay(In=^UE.Source.OnPlay)
   TriggerDelay_5 = UE.Trigger Delay(In=^UE.Source.OnPlay) {also {defaults {Delay Time=4}}}
   TriggerDelay_6 = UE.Trigger Delay(In=^UE.Source.OnPlay) {also {defaults {Delay Time=3}}}
   TriggerRouteTriggerRouteFloat6_0 = "TriggerRoute.Trigger Route (Float, 6):Float"(
    Set 1=@TriggerDelay_0.Out
    Set 5=@TriggerDelay_1.Out
    Set 4=@TriggerDelay_2.Out
    Set 0=@TriggerDelay_3.Out
    Set 3=@TriggerDelay_5.Out
    Set 2=@TriggerDelay_6.Out
   ) {also {defaults {Value 0=80, Value 1=160, Value 2=100, Value 3=60, Value 4=240, Value 5=120}}}
   n2 = Value.Value:Float(Set=@TriggerRouteTriggerRouteFloat6_0.On Set, Target Value=@TriggerRouteTriggerRouteFloat6_0.Value)
   MS_MIDIStreamSineGenerator_0 = /Harmonix/Tests/HarmonixMetasound/MS_MIDIStreamSineGenerator(MIDI Stream=HarmonixNodes.MIDIPlayer(
     Kill Voices On MIDI File Change=^MIDI Player.Kill Voices On MIDI File Change
     Kill Voices On Seek=^MIDI Player.Kill Voices On Seek
     MIDI File=^MIDI Player.MIDI File
     Pre-roll Bars=^MIDI Player.Pre-roll Bars
     Simple Loop=^MIDI Player.Simple Loop
     Speed=^MIDI Player.Speed
     MIDI Clock=HarmonixNodes.Metronome(Speed=^Metronome.Speed, Transport=@n1.Transport, Tempo=@n2.Output Value).MIDI Clock
     Transport=@n1.Transport
    ).MIDI Stream)
   TriggerDelay_4 = UE.Trigger Delay(In=^UE.Source.OnPlay) {also {defaults {Delay Time=8}}}
  ]
  links [
   {dst="@AudioOut", src=MS_MIDIStreamSineGenerator_0.Out}
   {dst="@UE.OutputFormat.Mono.Audio:0", src=MS_MIDIStreamSineGenerator_0.Out}
   {dst="@UE.Source.OneShot.OnFinished", src=TriggerDelay_4.Out}
  ]
 }]
properties {
 assetImportData=&self:AssetImportData
 sampleRate=48000
}
