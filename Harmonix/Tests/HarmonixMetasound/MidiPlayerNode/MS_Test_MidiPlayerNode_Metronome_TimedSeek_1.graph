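# One-shot mono MetaSound test source for the Harmonix MIDI Player node, exercising a timed seek during playback of a tempo-change scale MIDI file.
# OnPlay starts the transport and, through a Trigger Delay, later seeks it to a bar/beat target.
# The player's MIDI stream drives a test sine generator for the audio output.
# A MIDI clock timestamp trigger fires OnFinished at a bar shortly after the seek target.
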MetaSoundSource /Harmonix/Tests/HarmonixMetasound/MidiPlayerNode/MS_Test_MidiPlayerNode_Metronome_TimedSeek_1
inputs [
 MIDI Player.Kill Voices On MIDI File Change {type=Bool, access=Constructor}
 MIDI Player.Kill Voices On Seek {type=Bool, access=Constructor}
 MIDI Player.MIDI File {type=MIDIAsset}
 MIDI Player.Pre-roll Bars {type=Int32, access=Constructor}
 MIDI Player.Simple Loop {type=Bool}
 MIDI Player.Speed {type=Float}
 UE.Source.OnPlay {type=Trigger}
]
interfaces [UE.OutputFormat.Mono:1.0, UE.Source.OneShot:1.0, UE.Source:1.1]
outputs [
 AudioOut {type=Audio}
 "UE.OutputFormat.Mono.Audio:0" {type=Audio}
 UE.Source.OneShot.OnFinished {type=Trigger}
]
pages [Default {
  input_defaults {
   MIDI Player.Kill Voices On MIDI File Change=null
   MIDI Player.Kill Voices On Seek=null
   MIDI Player.MIDI File=&/scale_tempo_changes
   MIDI Player.Pre-roll Bars=8
   MIDI Player.Simple Loop=null
   MIDI Player.Speed=1
   UE.Source.OnPlay=null
  }
  nodes [
   TriggerDelay_0 = UE.Trigger Delay(In=^UE.Source.OnPlay)
   n1 = HarmonixNodes.MIDIPlayer(
    Kill Voices On MIDI File Change=^MIDI Player.Kill Voices On MIDI File Change
    Kill Voices On Seek=^MIDI Player.Kill Voices On Seek
    MIDI File=^MIDI Player.MIDI File
    Pre-roll Bars=^MIDI Player.Pre-roll Bars
    Simple Loop=^MIDI Player.Simple Loop
    Speed=^MIDI Player.Speed
    Transport=HarmonixNodes.TriggerToTransport(
     Play=^UE.Source.OnPlay
     Seek Target=HarmonixNodes.BarBeatToSeekTarget() {also {defaults {Bar=4, Beat=2}, version [0,1]}}.Seek Target
     Trigger Seek=@TriggerDelay_0.Out
    ).Transport
   )
   HarmonixNodesMidiClockTimestampTrigger_0 = HarmonixNodes.MidiClockTimestampTrigger(MIDI Clock=@n1.MIDI Clock, Music Timestamp=HarmonixNodes.MusicTimestampBuilder() {also {defaults {Bar=5}, version [0,1]}}.Music Timestamp)
   MS_MIDIStreamSineGenerator_0 = /Harmonix/Tests/HarmonixMetasound/MS_MIDIStreamSineGenerator(MIDI Stream=@n1.MIDI Stream)
  ]
  links [
   {dst="@UE.Source.OneShot.OnFinished", src=HarmonixNodesMidiClockTimestampTrigger_0.Trigger Out}
   {dst="@AudioOut", src=MS_MIDIStreamSineGenerator_0.Out}
   {dst="@UE.OutputFormat.Mono.Audio:0", src=MS_MIDIStreamSineGenerator_0.Out}
  ]
 }]
properties {
 assetImportData=&self:AssetImportData
 sampleRate=48000
}
